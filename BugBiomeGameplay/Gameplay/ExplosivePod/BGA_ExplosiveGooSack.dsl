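; Explosive goo sack actor for the bug biome that explodes when it dies.
; Its OnDied handler calls a multicast event that spawns a Niagara explosion, hides the actor and disables its collision.
; That event gathers actors within a scalable-float radius, runs them through FilterByLOS, and applies an explosion GameplayEffect to the results.
; The filter branches on player pawns and other goo sacks, using a line-of-sight trace from above the sphere.
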(event EventBeginPlay
  (EventDispatchers|BindEventtoOnDied (delegate Custom|On%20Died)))

(event Custom|On%20Died
  (params (DamagedActor :type Actor%20Object%20Reference) (Damage :type Float%20%28single-precision%29) (InstigatedBy :type Controller%20Object%20Reference) (DamageCauser :type Actor%20Object%20Reference) (HitLocation :type Vector) (FHitComponent :type Primitive%20Component%20Object%20Reference) (BoneName :type Name) (Momentum :type Vector))
  (CallFunction|OnDiedMulticast))

(event Custom|On%20Died%20Multicast
  (Niagara|SpawnSystematLocation "/Game/Weapons/FORT_GrenadeLaunchers/Effects/NS_GrenadeLauncher_Explosion_Athena_Pumpkin.NS_GrenadeLauncher_Explosion_Athena_Pumpkin" (Transformation|GetActorLocation))
  (Rendering|SetActorHiddenInGame true)
  (Collision|SetActorEnableCollision false)
  (bind _outactors (Collision|SphereOverlapActors (Transformation|GetActorLocation) (ScalableFloat|GetValueatLevel (Variables|Default|GetSF_DamageRadius)) (Utilities|Array|MakeArray "ObjectTypeQuery6" "ObjectTypeQuery1" "ObjectTypeQuery5" "ObjectTypeQuery3" "ObjectTypeQuery2" "ObjectTypeQuery4" "ObjectTypeQuery7") "/Script/Engine.Actor" (Utilities|Array|MakeArray self)))
  (bind _result (CallFunction|FilterbyLOS _outactors))
  (for _array_element _result
    (GameplayEffects|ApplyGameplayEffectToTarget (Ability|GetFortAbilitySystemComponent self) "/BugBiomeGameplay/Gameplay/ExplosivePod/GE_ExplosiveGooSack_Explosion.GE_ExplosiveGooSack_Explosion_C" (Ability|GetFortAbilitySystemComponent _array_element) 0.0 (GameplayEffects|MakeFortEffectContext (Ability|GetFortAbilitySystemComponent self) 0 0 self 0))))

(event Collision|EventActorBeginOverlap
  (params (OtherActor :type Actor%20Object%20Reference)))

(event EventTick
  (params (DeltaSeconds :type Float%20%28single-precision%29)))

(fn FilterByLOS
  (params (_Array :type Array%20of%20Actor%20Object%20References))
  (for _array_element _Array
    (Utilities|Casting|CastToPlayerPawn_Athena _array_element
      (:then
        _)
      (:CastFailed
        (Utilities|Casting|CastToBGA_ExplosiveGooSack _array_element
          (:then
            (bind _returnvalue (Fortnite|CheckLineOfSightToActorWithChannel (+ (Transformation|GetActorLocation) (* (Math|Vector|GetUpVector (Transformation|GetWorldRotation (Variables|Default|GetSphere))) 5e+01)) _array_element "ECC_GameTraceChannel16"))
            (if _returnvalue
              _))
          (:CastFailed
            _))
        (Utilities|Array|Add (LocalVariables|GetResult) _array_element))))
  (return (LocalVariables|GetResult)))

(fn ConstructionScript)
